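# Additive, unlit Niagara particle material without depth test that draws a capture ring's screen-edge glow toward a point offset from the actor position.
# View-vector alignment with that point shapes a powered glow that blends dark to bright color, tinted by particle color and faded by camera distance and facing.
# Distorted, panning radial fire noise breaks up the glow.
# An offset calculation re-centers geometry on the camera and throws it far away beyond distance or facing thresholds.
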Material /SparksEffects/Effects/GameplayFX/PerformancePositionFX/Materials/M_CaptureRing_Screen_Edge_Glow_Attached_Offset
graph {
 properties {
  bDisableDepthTest=true
  bUsedWithNiagaraMeshParticles=true
  bUsedWithNiagaraSprites=true
  blendMode=BLEND_Additive
  pixelDepthOffsetMode=PDOM_Legacy
  shadingModel=MSM_Unlit
  twoSided=true
  usedShadingModels=Unlit
 }
 nodes [
  n1 = WorldPosition()
  n2 = Add(A=CameraPositionWS(), B=@n1.XYZ)
  n3 = Multiply(A=CameraPositionWS(), B=VectorParameter() {parameterName=OffsetMultiplier, defaultValue=<rgba 1 1 1 0>}.RGB)
  n4 = Add(
   A=Multiply(A=ActorPositionWS(), B=VectorParameter() {parameterName=OffsetMultiplier, defaultValue=<rgba 1 1 1 0>}.RGB)
   B=VectorParameter() {parameterName=Offset, defaultValue=<rgba 0 0 5000 0>}.RGB
  )
  n5 = DynamicParameter() {paramNames [FadeOutDistanceMult, Param2, Param3, Param4]}
  n6 = ScalarParameter() {defaultValue=150000, parameterName=FadeDistance}
  n7 = Saturate(None=Divide(A=Distance(A=@n3, B=@n4), B=Multiply(A=@n5.FadeOutDistanceMult, B=@n6)))
  n8 = Floor(None=Multiply(A=@n7) {constB=1.01})
  n9 = Normalize(VectorInput=Subtract(A=@n3, B=@n4))
  n10 = Transform(None=Constant3Vector() {constant=<rgba 0 0 1 0>}) {transformSourceType=TRANSFORMSOURCE_Camera}
  n11 = Multiply(A=OneMinus(None=DotProduct(A=@n9, B=@n10)), B=ScalarParameter() {defaultValue=3, parameterName=PopOutThreshold}) {constB=2}
  n12 = OneMinus(None=Saturate(None=Floor(None=@n11)))
  Add(
   A=Subtract(A=Subtract(A=@n2, B=@n1.XYZ), B=ActorPositionWS())
   B=Multiply(A=Add(A=@n8, B=@n12), B=Constant3Vector() {constant=<rgba 0 0 -54654566400 0>})
  )
  Constant() {r=1}
  Constant() {r=150000}
  Constant()
  n13 = ParticleColor()
  n14 = VectorParameter() {parameterName=DarkColor, defaultValue=<rgba 0.474 0.1874 0.4156 1>}
  n15 = VectorParameter() {parameterName=BrightColor, defaultValue=<rgba 2 2 2 0>}
  n16 = DotProduct(A=CameraVectorWS(), B=@n9)
  n17 = ScalarParameter() {defaultValue=1.5, parameterName=MinEdgeGlowPower}
  n18 = ScalarParameter() {defaultValue=8, parameterName=MaxEdgeGlowPower}
  n19 = Multiply(A=Power(Base=Saturate(None=@n16), Exp=LinearInterpolate(A=@n17, B=@n18, Alpha=@n7))) {constB=2.75}
  n20 = Multiply(A=OneMinus(None=@n7), B=@n19)
  n21 = Saturate(None=@n20)
  MaterialFunctionCall(Emissive=Multiply(A=@n13.RGB, B=LinearInterpolate(A=@n14.RGB, B=@n15.RGB, Alpha=@n21))) {materialFunction=&MF_FX_EmissiveAdjustments}
  n22 = ActorPositionWS()
  n23 = VectorParameter() {parameterName=OffsetMultiplier, defaultValue=<rgba 1 1 1 0>}
  n24 = Subtract(A=@n3, B=Add(A=Multiply(A=@n22, B=@n23.RGB), B=Constant3Vector() {constant=<rgba 0 0 -7000 0>}))
  n25 = Saturate(None=DotProduct(A=CameraVectorWS(), B=Normalize(VectorInput=@n24)))
  n26 = Constant() {r=3}
  Max(A=@n19, B=Multiply(A=Power(Base=@n25, Exp=@n26)) {constB=0.5})
  Multiply(A=VectorParameter() {parameterName=Offset, defaultValue=<rgba 0 0 5000 0>}.RGB) {constB=-1}
  n27 = ComponentMask(None=MaterialFunctionCall() {materialFunction=&ScreenAlignedUVs}."X 100%, Y 100%") {g=true}
  OneMinus(None=@n27)
  n28 = OneMinus(None=DotProduct(A=Multiply(A=@n9) {constB=-1}, B=@n10))
  n29 = ScalarParameter() {defaultValue=3, parameterName=GlowFacingFadePower}
  n30 = Multiply(A=@n20, B=Multiply(A=@n19, B=Power(Base=@n28, Exp=@n29)))
  n31 = MaterialFunctionCall!(
   In Texture
   Random Color Texture
   In Texcoords=TextureCoordinate()
   Turn Time=Constant() {r=0.05}
   Distortion Tiling=Constant2Vector() {g=6, r=6}
   Distortion Amount
   Tiling
   Detail Factor
  ) {materialFunction=/Game/Effects/Niagara/MasterMaterials/Functions/Mojo_MF/MF_FX_TurnDistortion}
  n32 = TextureCoordinate()
  n33 = TextureSample!(
   UVs=Panner(Coordinate=MaterialFunctionCall!(Center, Maintain Aspect Ratio, Tiling, Offset, In Tex Coords=@n31.Out Base UV) {materialFunction=/Game/Effects/Niagara/MasterMaterials/Functions/Mojo_MF/MF_FX_Uniform_Radial_UV}.Out UV) {bFractionalPart=true, speedX=0.5, speedY=0.25}
   Tex
   "DDX(UVs)"=DDX(Value=@n32)
   "DDY(UVs)"=DDY(Value=@n32)
   Apply View MipBias
  ) {mipValueMode=TMVM_Derivative, texture=/Game/Effects/Fort_Effects/Textures/Fire/HF_FireNoisesPacked}
  Saturate(None=Multiply(A=Multiply(A=@n33.B) {constB=2}, B=Multiply(A=@n13.A, B=@n30)))
  Saturate(None=Multiply(A=Power(Base=@n27) {constExponent=9}) {constB=10})
  TextureCoordinate()
 ]
}
