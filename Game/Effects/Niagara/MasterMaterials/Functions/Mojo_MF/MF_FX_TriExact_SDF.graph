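# Material function that generates an exact signed distance field for a triangle defined by three 2D points, for use in FX materials.
# It finds the squared distance from the input coordinates to each clamped edge, pairs it with a winding-corrected cross-product side test, and takes the component-wise minimum.
# The negated square root times that sign gives the distance, negative inside.
# A second output passes it through MF_FX_Fill with an adjustable Fill Size.
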MaterialFunction /Game/Effects/Niagara/MasterMaterials/Functions/Mojo_MF/MF_FX_TriExact_SDF
graph {
 properties {description=Generates a triangle sdf based on 3 exact points}
 nodes [
  n1 = FunctionInput(Preview=Constant2Vector() {g=0.56, r=0.24}) {bUsePreviewValueAsDefault=true, inputName=p0, inputType=FunctionInput_Vector2, sortPriority=1}
  n2 = FunctionInput(Preview=TextureCoordinate()) {bUsePreviewValueAsDefault=true, description=texcoords, inputName=p, inputType=FunctionInput_Vector2}
  n3 = Subtract(A=@n2, B=@n1) {desc=v0}
  n4 = FunctionInput(Preview=Constant2Vector() {g=0.27, r=0.5}) {bUsePreviewValueAsDefault=true, inputName=p1, inputType=FunctionInput_Vector2, sortPriority=2}
  n5 = Subtract(A=@n4, B=@n1) {desc=e0}
  n6 = Divide(A=DotProduct(A=@n3, B=@n5), B=DotProduct(A=@n5, B=@n5))
  n7 = Subtract(A=@n3, B=Multiply(A=@n5, B=Saturate(None=@n6))) {desc=pq0}
  n8 = DotProduct(A=@n7, B=@n7)
  n9 = ComponentMask(None=@n3) {r=true}
  n10 = ComponentMask(None=@n5) {g=true}
  n11 = ComponentMask(None=@n3) {g=true}
  n12 = ComponentMask(None=@n5) {r=true}
  n13 = FunctionInput(Preview=Constant2Vector() {g=0.58, r=0.74}) {bUsePreviewValueAsDefault=true, inputName=p2, inputType=FunctionInput_Vector2, sortPriority=3}
  n14 = Subtract(A=@n1, B=@n13) {desc=e2}
  n15 = ComponentMask(None=@n14) {g=true}
  n16 = ComponentMask(None=@n14) {r=true}
  n17 = Sign(None=Subtract(A=Multiply(A=@n12, B=@n15), B=Multiply(A=@n10, B=@n16)))
  n18 = Multiply(A=@n17, B=Subtract(A=Multiply(A=@n9, B=@n10), B=Multiply(A=@n11, B=@n12)))
  n19 = Subtract(A=@n2, B=@n4) {desc=v1}
  n20 = Subtract(A=@n13, B=@n4) {desc=e1}
  n21 = Divide(A=DotProduct(A=@n19, B=@n20), B=DotProduct(A=@n20, B=@n20))
  n22 = Subtract(A=@n19, B=Multiply(A=@n20, B=Saturate(None=@n21))) {desc=pq1}
  n23 = DotProduct(A=@n22, B=@n22)
  n24 = ComponentMask(None=@n19) {r=true}
  n25 = ComponentMask(None=@n20) {g=true}
  n26 = ComponentMask(None=@n19) {g=true}
  n27 = ComponentMask(None=@n20) {r=true}
  n28 = Multiply(A=@n17, B=Subtract(A=Multiply(A=@n24, B=@n25), B=Multiply(A=@n26, B=@n27)))
  n29 = Subtract(A=@n2, B=@n13) {desc=v2}
  n30 = Divide(A=DotProduct(A=@n29, B=@n14), B=DotProduct(A=@n14, B=@n14))
  n31 = Subtract(A=@n29, B=Multiply(A=@n14, B=Saturate(None=@n30))) {desc=pq2}
  n32 = Subtract(A=Multiply(A=ComponentMask(None=@n29) {r=true}, B=@n15), B=Multiply(A=ComponentMask(None=@n29) {g=true}, B=@n16))
  n33 = Min(
   A=Min(A=AppendVector(A=@n8, B=@n18), B=AppendVector(A=@n23, B=@n28))
   B=AppendVector(A=DotProduct(A=@n31, B=@n31), B=Multiply(A=@n17, B=@n32))
  ) {desc=d}
  n34 = ComponentMask(None=@n33) {r=true}
  n35 = Multiply(A=Multiply(A=Constant() {r=-1}, B=SquareRoot(None=@n34)), B=Sign(None=ComponentMask(None=@n33) {g=true}))
  FunctionOutput(None=@n35)
  n36 = Constant() {r=0.01}
  FunctionOutput(None=MaterialFunctionCall!(
    x=@n35
    size=FunctionInput(Preview=@n36) {bUsePreviewValueAsDefault=true, inputName=Fill Size, inputType=FunctionInput_Scalar, sortPriority=4}
   ) {materialFunction=&/MF_FX_Fill}.Result) {outputName=Fill Result}
 ]
}
