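# Material function that computes an equilateral triangle distance field from a 2D "st" coordinate, which defaults to the texture coordinates.
# It remaps the coordinate to a centered -2 to 2 range, then mirrors X with Abs to share one slanted-edge term between both sides.
# A Max of that slanted term and the flat-edge term forms the triangle.
# The output is the raw gradient, with no threshold applied.
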MaterialFunction /Game/Effects/Niagara/MasterMaterials/Functions/Mojo_MF/MF_FX_Triangle_SDF_A
graph {nodes [
  n1 = Constant() {r=2}
  n2 = FunctionInput(Preview=TextureCoordinate()) {bUsePreviewValueAsDefault=true, inputName=st, inputType=FunctionInput_Vector2}
  n3 = Multiply(A=Subtract(A=Multiply(A=@n2, B=@n1), B=Constant() {r=1}), B=@n1)
  n4 = Multiply(A=Abs(None=ComponentMask(None=@n3) {r=true}), B=Constant() {r=0.866})
  n5 = ComponentMask(None=@n3) {g=true}
  n6 = Constant() {r=0.5}
  n7 = Multiply(A=@n5, B=@n6)
  n8 = Multiply(A=@n5, B=Constant() {r=-1})
  FunctionOutput(None=Max(A=Add(A=@n4, B=@n7), B=Multiply(A=@n8, B=@n6)))
 ]}
