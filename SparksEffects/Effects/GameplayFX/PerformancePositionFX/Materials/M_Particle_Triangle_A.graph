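# Translucent unlit Niagara particle material that draws an animated, particle-colored triangle for the performance position effect.
# Sine and cosine waves move three triangle points on rotating UVs, with particle random and a dynamic speed parameter available to vary timing.
# An SDF fill builds the mask, then edge protection and a step harden it for emissive and opacity.
# An optional switch computes camera-distance scale-up offset; an audio beat parameter is also declared.
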Material /SparksEffects/Effects/GameplayFX/PerformancePositionFX/Materials/M_Particle_Triangle_A
graph {
 properties {
  bAllowNegativeEmissiveColor=true
  bUsedWithNanite=true
  bUsedWithNiagaraMeshParticles=true
  bUsedWithNiagaraSprites=true
  blendMode=BLEND_Translucent
  pixelDepthOffsetMode=PDOM_Legacy
  shadingModel=MSM_Unlit
  usedShadingModels=Unlit
 }
 nodes [
  n1 = Multiply(A=Time(), B=NamedRerouteUsage())
  n2 = ParticleRandom()
  Add(A=@n1, B=@n2)
  Constant() {r=2}
  Constant() {desc=speed Mult, r=0.25}
  n3 = ParticleColor()
  n4 = MaterialFunctionCall!(TexCoords=NamedRerouteUsage(), Edge Sharpness=Constant() {r=10}, Edge Subtract) {materialFunction=/Game/Effects/Niagara/MasterMaterials/Functions/Mojo_MF/MF_FX_UV_EdgeMask}
  n5 = Step(X=Multiply(A=Saturate(None=@n4.1-Edge Mask), B=NamedRerouteUsage()) {desc="protect sprite  UV edges"}) {constX=0.5, constY=0.5, desc=test}
  MaterialFunctionCall(Emissive=Multiply(A=@n5, B=@n3.RGB)) {materialFunction=&MF_FX_EmissiveAdjustments}
  n6 = Frac(None=@n1)
  Multiply(A=@n6, B=@n2)
  n7 = NamedRerouteUsage()
  n8 = MaterialFunctionCall!(p=@n7, p0=NamedRerouteUsage(), p1=NamedRerouteUsage(), p2=NamedRerouteUsage(), Fill Size) {materialFunction=/Game/Effects/Niagara/MasterMaterials/Functions/Mojo_MF/MF_FX_TriExact_SDF}
  n9 = MaterialFunctionCall!(
   x=MaterialFunctionCall!(st=@n7) {materialFunction=/Game/Effects/Niagara/MasterMaterials/Functions/Mojo_MF/MF_FX_Triangle_SDF_A}.Result
   size=Constant() {r=0.5}
  ) {materialFunction=/Game/Effects/Niagara/MasterMaterials/Functions/Mojo_MF/MF_FX_Fill}
  n10 = Constant() {desc=debug 0 or 1}
  NamedRerouteDeclaration(None=Saturate(None=LinearInterpolate(A=@n8.Fill Result, B=@n9.Result, Alpha=@n10))) {name=TriangleMask, nodeColor=<rgba 0 0.5312 1 1>}
  n11 = Multiply(A=NamedRerouteUsage(), B=Constant() {r=0.2})
  n12 = MaterialFunctionCall(
   Input=Cosine(None=@n11)
   Input Low=Constant() {r=-1}
   Input High=Constant() {r=1}
   Target Low=Constant() {r=-0.2}
   Target High=Constant() {r=0.2}
  ) {materialFunction=&RemapValueRange}
  n13 = NamedRerouteUsage()
  n14 = Constant() {r=1.3}
  n15 = MaterialFunctionCall(
   Input=Sine(None=Multiply(A=@n13, B=@n14)) {period=6.28}
   Input Low=Constant() {r=-1}
   Input High=Constant() {r=1}
   Target Low=Constant() {r=-0.3}
   Target High=Constant() {r=0.3}
  ) {materialFunction=&RemapValueRange}
  NamedRerouteDeclaration(None=Add(A=Constant() {r=0.5}, B=AppendVector(A=@n12.Result, B=@n15.Result))) {name=P2, nodeColor=<rgba 1 0.0234 0 1>}
  NamedRerouteDeclaration(None=CollectionParameter() {collection=/FMAudioReactiveData/AudioAnalysis/MPC_Sparks_Audio_Parameters, parameterName=BeatLifetime}) {name=Beat Raw, nodeColor=<rgba 0 1 0.4844 1>}
  n16 = NamedRerouteUsage()
  n17 = Constant() {r=0.5}
  n18 = MaterialFunctionCall(
   Input=Sine(None=Multiply(A=@n16, B=@n17)) {period=6.28}
   Input Low=Constant() {r=-1}
   Input High=Constant() {r=1}
   Target Low=Constant() {r=-0.2}
   Target High=Constant() {r=0.2}
  ) {materialFunction=&RemapValueRange}
  n19 = NamedRerouteUsage()
  n20 = Constant() {r=1}
  n21 = MaterialFunctionCall(
   Input=Sine(None=Multiply(A=@n19, B=@n20)) {period=6.28}
   Input Low=Constant() {r=-1}
   Input High=Constant() {r=1}
   Target Low=Constant() {r=-0.2}
   Target High=Constant() {r=0.2}
  ) {materialFunction=&RemapValueRange}
  NamedRerouteDeclaration(None=Add(A=Constant() {r=0.5}, B=AppendVector(A=@n18.Result, B=@n21.Result))) {name=P1, nodeColor=<rgba 1 0 0.4453 1>}
  n22 = NamedRerouteUsage()
  n23 = Constant() {r=1}
  n24 = MaterialFunctionCall(
   Input=Sine(None=Multiply(A=@n22, B=@n23)) {period=6.28}
   Input Low=Constant() {r=-1}
   Input High=Constant() {r=1}
   Target Low=Constant() {r=-0.2}
   Target High=Constant() {r=0.2}
  ) {materialFunction=&RemapValueRange}
  n25 = NamedRerouteUsage()
  n26 = Constant() {r=0.8}
  n27 = MaterialFunctionCall(
   Input=Sine(None=Multiply(A=@n25, B=@n26)) {period=6.28}
   Input Low=Constant() {r=-1}
   Input High=Constant() {r=1}
   Target Low=Constant() {r=-0.2}
   Target High=Constant() {r=0.2}
  ) {materialFunction=&RemapValueRange}
  NamedRerouteDeclaration(None=Add(A=Constant() {r=0.5}, B=AppendVector(A=@n24.Result, B=@n27.Result))) {name=P0, nodeColor=<rgba 1 0 0.2109 1>}
  n28 = Time()
  n29 = NamedRerouteUsage()
  n30 = ParticleRandom()
  NamedRerouteDeclaration(None=Add(A=Multiply(A=@n28, B=@n29), B=Multiply(A=@n30) {constB=100})) {name=timeScaled, nodeColor=<rgba 1 0.3984 0 1>}
  NamedRerouteDeclaration(None=MaterialFunctionCall(UVs=TextureCoordinate(), "Rotation Angle (0-1)"=@n6) {materialFunction=&CustomRotator}.Rotated Values) {name=baseUV, nodeColor=<rgba 0 0.3672 1 1>}
  NamedRerouteDeclaration(None=DynamicParameter() {paramNames [RandomTriangleSpeedMult, Param2, Param3, Param4]}.RandomTriangleSpeedMult) {name=RandomTriangleSpeedMult, nodeColor=<rgba 0.6406 1 0 1>}
  n31 = ObjectPositionWS()
  n32 = WorldPosition()
  n33 = Multiply(A=ScalarParameter() {defaultValue=10, parameterName=MaxScaleIncrease}, B=Multiply(A=Subtract(A=@n31, B=@n32.XYZ)) {constB=-1}) {constB=-1}
  n34 = Distance(A=@n31, B=CameraPositionWS())
  n35 = ScalarParameter() {defaultValue=2000, parameterName=DistanceOffset}
  n36 = Clamp(None=Divide(A=Subtract(A=@n34, B=@n35), B=ScalarParameter() {defaultValue=30000, parameterName=ScaleUpDistance}))
  NamedRerouteDeclaration(None=StaticSwitchParameter(True=LinearInterpolate(B=@n33, Alpha=@n36) {constB=0}, False=Constant()) {parameterName=Use Distance Fade}) {name=WPO_Out, nodeColor=<rgba 0.2891 1 0 1>}
  NamedRerouteUsage()
  NamedRerouteUsage()
  Reroute(None=Multiply(A=@n5, B=@n3.A))
  Sine(None=@n11) {period=6.28}
  Step(Y=ParticleRandom()) {constX=0.5}
 ]
}
